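# Sound cue for the Statue Pose emote music in Athena, made of an intro and a looping body.
# A mixer starts both branches together: the one-shot intro plays immediately, while the looping wave waits behind a fixed two-second delay.
# Each branch overrides its sound class to the first-person emote music class, matching the cue's own.
# The cue uses Athena emote attenuation and music concurrency, with a send to the music audio-analysis submix.
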SoundCue /Game/Athena/Sounds/Emotes/StatuePose/Emote_StatuePose_Cue
root=Mixer_1
nodes [
 {id=Delay_1, children [WavePlayer_6], properties {delayMax=2, delayMin=2}}
 {id=Mixer_1, children [SoundClass_2,SoundClass_3], properties {inputVolume [1,1]}}
 {id=SoundClass_2, children [WavePlayer_4], properties {soundClassOverride=/Game/Sounds/FortSoundClassesAndModes/Fort_Music_Emote_1P_PSM}}
 {id=SoundClass_3, children [Delay_1], properties {soundClassOverride=&Fort_Music_Emote_1P_PSM}}
 {id=WavePlayer_4, properties {soundWaveAssetPtr=&/Emote_StatuePose_Intro_01}}
 {id=WavePlayer_6, properties {bLooping=true, soundWaveAssetPtr=&/Emote_StatuePose_01}}
]
properties {
 attenuationSettings=&Fort_Emotes_Athena_Att
 maxDistance=1600
 soundClassObject=&Fort_Music_Emote_1P_PSM
 volumeMultiplier=0.9
 concurrencySet [/Game/Sounds/Concurrency_Rules/Athena_EmoteMusic_1P_Conc]
 soundSubmixSends [{minSendLevel=0, soundSubmix=&Fortnite_AudioAnalysis_Music}]
}
